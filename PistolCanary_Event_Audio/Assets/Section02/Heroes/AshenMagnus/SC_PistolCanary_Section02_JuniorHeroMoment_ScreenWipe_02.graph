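# Sound cue for the screen-wipe moment of a junior hero (Ashen Magnus) in Section 02 of the PistolCanary event.
# A single wave player feeds the output directly, playing the PistolCanary character intro 03 wave with no randomization or modulation.
# It uses the large point-of-interest attenuation preset, the Fort events sound class and a reduced volume.
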SoundCue /PistolCanary_Event_Audio/Assets/Section02/Heroes/AshenMagnus/SC_PistolCanary_Section02_JuniorHeroMoment_ScreenWipe_02
root=WavePlayer_4
nodes [{id=WavePlayer_4, properties {soundWaveAssetPtr=/PistolCanary_Event_Audio/Assets/Section02/wav/PistolCanary_Character_Intro_03}}]
properties {
 attenuationSettings=/PistolCanary_Audio/Assets/Settings/ATT_PistolCanary_POI_Large
 maxDistance=40000
 soundClassObject=&Fort_Events
 volumeMultiplier=0.6
}
